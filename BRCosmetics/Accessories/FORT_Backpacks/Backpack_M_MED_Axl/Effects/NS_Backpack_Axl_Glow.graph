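# Niagara glow effect for the Axl backpack cosmetic, built only from orbiting light particles with no visible sprites.
# Two near-identical emitters each burst one particle, the second slightly delayed: one fixed pink, the other tinted by User.LightVariant.
# Each light rotates around a point while a curve fades its color in and out over its roughly four-second lifetime.
# Extra spawn bursts in both emitters are disabled.
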NiagaraSystem /BRCosmetics/Accessories/FORT_Backpacks/Backpack_M_MED_Axl/Effects/NS_Backpack_Axl_Glow
dependencies {usedDataInterfaces [&NiagaraDataInterfaceCurve], usedDynamicInputs [], usedModules [&SystemState]}
emitters {
 FlareSameParticlePink {
  emitterUpdateScript [
   &EmitterState()
   &SpawnBurst_Instantaneous(Spawn Count=1, Use Spawn Probability=false)
   &SpawnBurst_Instantaneous(Spawn Count=1, Spawn Time=0.1667, Use Spawn Probability=false) {enabled=false, module=SpawnBurst_Instantaneous002}
   &SpawnBurst_Instantaneous(Spawn Count=1, Spawn Time=0.3333, Use Spawn Probability=false) {enabled=false, module=SpawnBurst_Instantaneous001}
  ]
  particleSpawnScript [
   /Niagara/Modules/Spawn/Initialization/InitializeParticle(
    Color=<rgba 2 0.058 1.012 1>
    Lifetime=4.2
    Module.Write Color=true
    Module.Write Lifetime=true
    Module.Write Mass=true
    Module.Write Material Random=true
    Module.Write Position=true
    Module.Write Scale=false
    Module.Write SpriteRotation=false
    Module.Write SpriteSize=false
    Position=&SimulationPosition()
    Sprite Size=&Vector2DFromFloat(Value=10)
   )
   /Game/Effects/Niagara/Modules/Groups/FN_SetSpawnGroup_v1/FN_SetSpawnGroup(Emitter Bone Count=2)
  ]
  particleUpdateScript [
   &ParticleState()
   &ScaleColor(
    Scale RGB=&VectorFromFloat(Value=&FloatFromCurve(FloatCurve=DataInterface() {propertyValues {Curve {keys [
          {interpMode=RCIM_Cubic, leaveTangent=11.5974, tangentMode=RCTM_Break, time=0, value=0}
          {arriveTangent=-0.0426, interpMode=RCIM_Cubic, leaveTangent=-0.0426, tangentMode=RCTM_User, time=0.15, value=1}
          {arriveTangent=0.0145, interpMode=RCIM_Cubic, leaveTangent=0.0145, tangentMode=RCTM_User, time=0.8, value=1}
          <at 1 0>
         ]}}}))
    ScaleA=false
    ScaleRGB=true
   )
   &RotateAroundPoint(Radius=10, Rotation Rate=2.125)
  ]
  rendererClasses [&NiagaraLightRendererProperties]
  renderers [{rendererClass=&NiagaraLightRendererProperties, rendererIndex=0}]
 }
 FlareSameParticleVariable {
  emitterUpdateScript [
   &EmitterState()
   &SpawnBurst_Instantaneous(Spawn Count=1, Spawn Time=0.0417, Use Spawn Probability=false) {enabled=false}
   &SpawnBurst_Instantaneous(Spawn Count=1, Spawn Time=0.2083, Use Spawn Probability=false) {module=SpawnBurst_Instantaneous002}
   &SpawnBurst_Instantaneous(Spawn Count=1, Spawn Time=0.375, Use Spawn Probability=false) {enabled=false, module=SpawnBurst_Instantaneous001}
  ]
  particleSpawnScript [
   &InitializeParticle(
    Color=$User.LightVariant
    Lifetime=4
    Module.Write Color=true
    Module.Write Lifetime=true
    Module.Write Mass=true
    Module.Write Material Random=true
    Module.Write Position=true
    Module.Write Scale=false
    Module.Write SpriteRotation=false
    Module.Write SpriteSize=false
    Position=&SimulationPosition()
    Sprite Size=&Vector2DFromFloat(Value=10)
   )
   &FN_SetSpawnGroup(Emitter Bone Count=2)
  ]
  particleUpdateScript [
   &ParticleState()
   &ScaleColor(
    Scale RGB=&VectorFromFloat(Value=&FloatFromCurve(FloatCurve=DataInterface() {propertyValues {Curve {keys [
          {interpMode=RCIM_Cubic, leaveTangent=11.5974, tangentMode=RCTM_Break, time=0, value=0}
          {arriveTangent=-0.0426, interpMode=RCIM_Cubic, leaveTangent=-0.0426, tangentMode=RCTM_User, time=0.15, value=1}
          {arriveTangent=0.0145, interpMode=RCIM_Cubic, leaveTangent=0.0145, tangentMode=RCTM_User, time=0.8, value=1}
          <at 1 0>
         ]}}}))
    ScaleA=false
    ScaleRGB=true
   )
   &RotateAroundPoint(Radius=10, Rotation Rate=2.125)
  ]
  rendererClasses [&NiagaraLightRendererProperties]
  renderers [{rendererClass=&NiagaraLightRendererProperties, rendererIndex=0}]
 }
}
summary {
 systemName=NS_Backpack_Axl_Glow
 emitters [
  {emitterName=FlareSameParticlePink, rendererClasses [&NiagaraLightRendererProperties]}
  {emitterName=FlareSameParticleVariable, rendererClasses [&NiagaraLightRendererProperties]}
 ]
 userVariables [
  User.bIsFrontEndPreview {type=NiagaraBool}
  User.FEVec {type=Vector3f, description="", defaultValue {struct=Vector3f, value=<0 0 1000>}}
  User.LightVariant {type=LinearColor, description="", defaultValue {struct=LinearColor, value=<rgba 1 1 1 1>}}
  User.PlayerSpeed {type=NiagaraFloat, defaultValue=0, description=""}
  User.RibbonIDOffset {type=NiagaraInt32}
  User.SkeletalMesh {
   type=NiagaraDataInterfaceSkeletalMesh
   defaultValue {
    struct=NiagaraExt_VariableValue_DataInterface
    value {dataInterface=&self:NiagaraDataInterfaceSkeletalMesh_0}
   }
  }
  User.SpawnRate {type=NiagaraFloat, defaultValue=360, description=""}
 ]
}
