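# Skinned material for the Axl backpack cosmetic that adds animated chasing lights to its emissive, between the shared character PreFX and PostFX functions.
# The FX mask's green channel selects two light zones, with static emissive added to the vertical one.
# Arctangent of pre-skinned, bounds-normalized position gives radial coordinates in XY and XZ without UVs.
# These are panned, blended by zone and wrapped into repeating lights with fading, hue-shifted tails.
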Material /BRCosmetics/Accessories/FORT_Backpacks/Backpack_M_MED_Axl/Materials/M_Backpack_M_MED_Axl
graph {
 properties {
  bUseMaterialAttributes=true
  bUsedWithMeshDeformer=true
  bUsedWithSkeletalMesh=true
  pixelDepthOffsetMode=PDOM_Legacy
  previewMesh=/BRCosmetics/Accessories/FORT_Backpacks/Backpack_M_MED_Axl/Meshes/M_MED_Axl_Pack
  usedShadingModels=Default Lit
 }
 nodes [
  n1 = MaterialFunctionCall() {materialFunction=&BaseFN_CharacterFunction_PreFX}
  n2 = GetMaterialAttributes(None=@n1.Result) {attributeGetTypes [#1]}
  n3 = TextureObjectParameter() {
   group=Skin FX
   parameterName=SkinFX_Mask
   samplerType=SAMPLERTYPE_Masks
   texture=/BRCosmetics/Accessories/FORT_Backpacks/Backpack_M_MED_Axl/Textures/T_M_MED_Axl_Backpack_FX
  }
  n4 = TextureSample(Tex=@n3) {samplerType=SAMPLERTYPE_Masks}
  n5 = SphereMask(A=@n4.G, B=Constant() {r=0.5}, Radius=Constant() {r=0.4}, Hardness=Constant() {r=1})
  n6 = VectorParameter() {group=Axl, parameterName=StaticEmissiveVertical, defaultValue=<rgba 0.2858 0.4818 0.5417 0>}
  n7 = ScalarParameter() {group=Axl, parameterName=LightIntensity}
  n8 = Add(A=@n2.Emissive Color, B=Multiply(A=Multiply(A=@n5, B=@n6.RGB), B=@n7))
  n9 = SphereMask(A=@n4.G, B=Constant() {r=1}, Radius=Constant() {r=0.4}, Hardness=Constant() {r=1})
  n10 = PreSkinnedPosition()
  n11 = Divide(
   A=Add(A=VectorParameter() {group=Axl, parameterName=RadialUVOffset, defaultValue=<rgba 0 17 0 0>}.RGB, B=VertexInterpolator(VS=@n10).PS)
   B=PreSkinnedLocalBounds().Extents
  )
  n12 = Divide(A=Arctangent2Fast(Y=ComponentMask(None=@n11) {g=true}, X=ComponentMask(None=@n11) {r=true}), B=Constant() {r=3.1416})
  n13 = Multiply(A=Multiply(A=Add(A=@n12)) {constB=0.5}, B=ScalarParameter() {defaultValue=3, group=Axl, parameterName=LightCount})
  n14 = ScalarParameter() {defaultValue=-6, group=Axl, parameterName=Speed}
  n15 = PreSkinnedPosition()
  n16 = Divide(
   A=Add(A=VectorParameter() {group=Axl, parameterName=RadialUVOffset2, defaultValue=<rgba 0 0 -2 0>}.RGB, B=VertexInterpolator(VS=@n15).PS)
   B=PreSkinnedLocalBounds().Extents
  )
  n17 = Divide(A=Arctangent2Fast(Y=ComponentMask(None=@n16) {b=true}, X=ComponentMask(None=@n16) {r=true}), B=Constant() {r=3.1416})
  n18 = Multiply(A=Multiply(A=Add(A=@n17)) {constB=0.5}, B=ScalarParameter() {defaultValue=3, group=Axl, parameterName=LightCountVertical})
  n19 = ScalarParameter() {defaultValue=-6, group=Axl, parameterName=SpeedVertical}
  n20 = Frac(None=LinearInterpolate(A=Panner(Coordinate=@n13, Speed=@n14) {speedY=1}, B=Panner(Coordinate=@n18, Speed=@n19) {speedY=1}, Alpha=@n5))
  n21 = Distance(A=@n20, B=Constant())
  n22 = Multiply(
   A=SphereMask(
    A=@n20
    B=Constant() {r=0.5}
    Radius=ScalarParameter() {defaultValue=1, group=Axl, parameterName=LightRadius}
    Hardness=ScalarParameter() {group=Axl, parameterName=LightHardness}
   ) {attenuationRadius=0.5, hardnessPercent=0}
   B=Power(Base=Saturate(None=@n21), Exp=ScalarParameter() {defaultValue=9, group=Axl, parameterName=TailFade})
  )
  n23 = Power(Base=OneMinus(None=@n21), Exp=ScalarParameter() {defaultValue=1, group=Axl, parameterName=HueShiftTailPow})
  n24 = ScalarParameter() {defaultValue=1, group=Axl, parameterName=HueShift}
  n25 = MaterialFunctionCall(
   Hue Shift Percentage=Frac(None=Multiply(A=@n23, B=@n24))
   Texture=VectorParameter() {group=Axl, parameterName=LightColor, defaultValue=<rgba 0 1 1 0>}.RGB
  ) {materialFunction=&HueShift}
  n26 = Multiply(A=@n7, B=Multiply(A=Add(A=@n9, B=@n5), B=Multiply(A=@n22, B=@n25.Result)))
  MaterialFunctionCall(
   AO=@n1.AO
   Distance Blend=@n1.AO
   Fresnel=@n1.AO
   Material=SetMaterialAttributes(MaterialAttributes=@n2.MaterialAttributes, Emissive Color=Add(A=@n8, B=@n26)) {attributeSetTypes [#1]}
  ) {materialFunction=&BaseFN_CharacterFunction_PostFX}
 ]
}
