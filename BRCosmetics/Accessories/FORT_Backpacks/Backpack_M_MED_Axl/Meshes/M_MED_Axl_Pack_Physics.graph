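# Physics asset for the M_MED_Axl backpack cosmetic, giving its rim and hanging net simulated secondary motion.
# Box bodies chain from pack through rimPivot and three net segments, then fan out to nine small netEnd bodies attached to net3.
# The rim hinge locks twist and one swing axis with slight play on the other, while angular limits loosen down the net so the ends swing most freely.
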PhysicsAsset /BRCosmetics/Accessories/FORT_Backpacks/Backpack_M_MED_Axl/Meshes/M_MED_Axl_Pack_Physics
bodies {
 net1 [{shapeType=Box, extentX=4, extentY=17, extentZ=17, center=<2.1256 0 0>}]
 net2 [{shapeType=Box, extentX=4, extentY=14, extentZ=14, center=<1.9588 0 0>}]
 net3 [{shapeType=Box, extentX=3.5, extentY=10.5, extentZ=10.5, center=<1.6737 0 0>}]
 netEnd1 [{shapeType=Box, extentX=2, extentY=1, extentZ=2, center=<1 0 0>}]
 netEnd2 [{shapeType=Box, extentX=2.1186, extentY=0.5287, extentZ=2.0628, center=<1.0266 -0.0929 0.0965>}]
 netEnd3 [{shapeType=Box, extentX=2, extentY=1, extentZ=2, center=<1 0 0>}]
 netEnd4 [{shapeType=Box, extentX=2, extentY=1, extentZ=2, center=<1 0 0>}]
 netEnd5 [{shapeType=Box, extentX=2, extentY=1, extentZ=2, center=<1 0 0>}]
 netEnd6 [{shapeType=Box, extentX=2.5629, extentY=0.8473, extentZ=1.9091, center=<1.1403 -0.1531 -0.2283>}]
 netEnd7 [{shapeType=Box, extentX=2.5647, extentY=0.85, extentZ=2.2406, center=<1.0929 -0.0442 0.0156>}]
 netEnd8 [{shapeType=Box, extentX=2, extentY=1, extentZ=2, center=<1 0 0>}]
 netEnd9 [{shapeType=Box, extentX=2.2551, extentY=0.8435, extentZ=2.0869, center=<1.0041 -0.1627 0.2169>}]
 pack [{shapeType=Box, extentX=0.5, extentY=0.5, extentZ=0.5, center=<0.0019 -2.8389 3.8235>}]
 rimPivot [{shapeType=Box, extentX=23, extentY=23, extentZ=5.5, center=<10.5 0 -2.5>}]
}
constraints [
 {bone1Name=rimPivot, bone2Name=pack, swing1LimitDegrees=0, swing1Motion=Locked, swing2LimitDegrees=3, twistLimitDegrees=0, twistMotion=Locked}
 {bone1Name=net1, bone2Name=rimPivot, swing1LimitDegrees=5, swing2LimitDegrees=5, twistLimitDegrees=1}
 {bone1Name=net2, bone2Name=net1, swing1LimitDegrees=7, swing2LimitDegrees=7, twistLimitDegrees=1}
 {bone1Name=net3, bone2Name=net2, swing1LimitDegrees=9, swing2LimitDegrees=9, twistLimitDegrees=1}
 {bone1Name=netEnd1, bone2Name=net3, swing1LimitDegrees=30, swing2LimitDegrees=30, twistLimitDegrees=5}
 {bone1Name=netEnd2, bone2Name=net3, swing1LimitDegrees=35, swing2LimitDegrees=30, twistLimitDegrees=5}
 {bone1Name=netEnd3, bone2Name=net3, swing1LimitDegrees=30, swing2LimitDegrees=35, twistLimitDegrees=5}
 {bone1Name=netEnd4, bone2Name=net3, swing1LimitDegrees=30, swing2LimitDegrees=30, twistLimitDegrees=5}
 {bone1Name=netEnd5, bone2Name=net3, swing1LimitDegrees=35, swing2LimitDegrees=30, twistLimitDegrees=5}
 {bone1Name=netEnd6, bone2Name=net3, swing1LimitDegrees=30, swing2LimitDegrees=35, twistLimitDegrees=5}
 {bone1Name=netEnd7, bone2Name=net3, swing1LimitDegrees=30, swing2LimitDegrees=30, twistLimitDegrees=5}
 {bone1Name=netEnd8, bone2Name=net3, swing1LimitDegrees=35, swing2LimitDegrees=30, twistLimitDegrees=5}
 {bone1Name=netEnd9, bone2Name=net3, swing1LimitDegrees=30, swing2LimitDegrees=35, twistLimitDegrees=5}
]
